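# Neon tube material for a stage venue, with glow shaped by axis-aligned Fresnel, brightness and audio-reactive animation.
# Color is a single parameter or vertex RGB-selected palette with flicker masks; a switch also offers stage show colors.
# Layered sines flicker on beat or on position-offset time, and optional vertex-alpha bands form frame animation synced to measure or time.
# Ray-traced lighting swaps Fresnel for a flat multiplier, and mobile and Switch get separate tuning.
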Material /SparksMaps/Environment/Venue_A/S7/Sets/Neon_Lights/Meshes/M_NeonTube_StageShow
graph {
 expressions {
  MaterialExpressionAdd_0 {
   class=MaterialExpressionAdd
   inputs [
    A {from=MaterialExpressionSwitch_0}
    B {from=MaterialExpressionNamedRerouteUsage_11}
   ]
  }
  MaterialExpressionAdd_1 {
   class=MaterialExpressionAdd
   inputs [
    A {from=MaterialExpressionComponentMask_0}
    B {from=MaterialExpressionComponentMask_1}
   ]
  }
  MaterialExpressionAdd_16 {
   class=MaterialExpressionAdd
   inputs [
    A {from=MaterialExpressionConstantBiasScale_3}
    B {from=""}
   ]
   properties {constB=0.1}
  }
  MaterialExpressionAdd_2 {
   class=MaterialExpressionAdd
   inputs [
    A {from=MaterialExpressionMultiply_11}
    B {from=MaterialExpressionAdd_1}
   ]
  }
  MaterialExpressionAdd_5 {
   class=MaterialExpressionAdd
   inputs [
    A {from=MaterialExpressionSwitch_1}
    B {from=MaterialExpressionNamedRerouteUsage_12}
   ]
  }
  MaterialExpressionAdd_6 {
   class=MaterialExpressionAdd
   inputs [
    A {from=MaterialExpressionSwitch_2}
    B {from=MaterialExpressionNamedRerouteUsage_2}
   ]
  }
  MaterialExpressionCollectionParameter_0 {
   class=MaterialExpressionCollectionParameter
   inputs []
   properties {collection=/FMAudioReactiveData/AudioAnalysis/MPC_Sparks_Audio_Parameters, parameterName=MeasureLifetime}
  }
  MaterialExpressionCollectionParameter_1 {
   class=MaterialExpressionCollectionParameter
   inputs []
   properties {collection=&MPC_Sparks_Audio_Parameters, parameterName=BeatLifetime}
  }
  MaterialExpressionComponentMask_0 {class=MaterialExpressionComponentMask, inputs [None {from=MaterialExpressionObjectPositionWS_1}], properties {r=true}}
  MaterialExpressionComponentMask_1 {class=MaterialExpressionComponentMask, inputs [None {from=MaterialExpressionObjectPositionWS_1}], properties {g=true}}
  MaterialExpressionConstant3Vector_0 {class=MaterialExpressionConstant3Vector, properties {constant=<rgba 0 1 0 0>}}
  MaterialExpressionConstant3Vector_2 {class=MaterialExpressionConstant3Vector, properties {constant=<rgba 1 1 1 0>}}
  MaterialExpressionConstantBiasScale_3 {
   class=MaterialExpressionConstantBiasScale
   inputs [None {from=MaterialExpressionLinearInterpolate_5}]
   properties {bias=4, desc=+ *, scale=2}
  }
  MaterialExpressionFloor_1 {class=MaterialExpressionFloor, inputs [None {from=MaterialExpressionMultiply_9, output=""}], properties {}}
  MaterialExpressionFrac_1 {class=MaterialExpressionFrac, inputs [None {from=MaterialExpressionMultiply_10, output=""}], properties {}}
  MaterialExpressionLinearInterpolate_10 {
   class=MaterialExpressionLinearInterpolate
   inputs [
    A {from=MaterialExpressionMultiply_12}
    B {from=MaterialExpressionMultiply_13}
    Alpha {from=MaterialExpressionVertexColor_1, output=B}
   ]
   properties {desc=Blue}
  }
  MaterialExpressionLinearInterpolate_11 {
   class=MaterialExpressionLinearInterpolate
   inputs [
    A {from=MaterialExpressionMultiply_16}
    B {from=MaterialExpressionMultiply_17}
    Alpha {from=MaterialExpressionVertexColor_1, output=B}
   ]
   properties {desc=Blue}
  }
  MaterialExpressionLinearInterpolate_12 {
   class=MaterialExpressionLinearInterpolate
   inputs [
    A {from=MaterialExpressionMultiply_14}
    B {from=MaterialExpressionMultiply_15}
    Alpha {from=MaterialExpressionVertexColor_1, output=B}
   ]
   properties {desc=Blue}
  }
  MaterialExpressionLinearInterpolate_5 {
   class=MaterialExpressionLinearInterpolate
   inputs [
    A {from=MaterialExpressionSine_4}
    B {from=MaterialExpressionSine_5}
    Alpha {from=MaterialExpressionSine_6}
   ]
  }
  MaterialExpressionLinearInterpolate_7 {
   class=MaterialExpressionLinearInterpolate
   inputs [
    A {from=MaterialExpressionLinearInterpolate_9}
    B {from=MaterialExpressionLinearInterpolate_8}
    Alpha {from=MaterialExpressionVertexColor_1, output=R}
   ]
   properties {desc=Red}
  }
  MaterialExpressionLinearInterpolate_8 {
   class=MaterialExpressionLinearInterpolate
   inputs [
    A {from=MaterialExpressionLinearInterpolate_11}
    B {from=MaterialExpressionMultiply_18}
    Alpha {from=MaterialExpressionVertexColor_1, output=G}
   ]
   properties {desc=Green}
  }
  MaterialExpressionLinearInterpolate_9 {
   class=MaterialExpressionLinearInterpolate
   inputs [
    A {from=MaterialExpressionLinearInterpolate_10}
    B {from=MaterialExpressionLinearInterpolate_12}
    Alpha {from=MaterialExpressionVertexColor_1, output=G}
   ]
   properties {desc=Green}
  }
  MaterialExpressionMaterialFunctionCall_0 {
   class=MaterialExpressionMaterialFunctionCall
   inputs [
    Axis {from=MaterialExpressionTransform_0}
    Exponent {from=MaterialExpressionShadingPathSwitch_1}
    Normal {from=""}
   ]
   properties {materialFunction=/Game/Effects/Fort_Effects/Materials/_Functions/MF_AxisAlignedFrensnel}
  }
  MaterialExpressionMaterialFunctionCall_2 {
   class=MaterialExpressionMaterialFunctionCall
   inputs [
    A {from=MaterialExpressionMaterialFunctionCall_8, output=Primary}
    B {from=MaterialExpressionMaterialFunctionCall_8, output=Primary}
    C {from=MaterialExpressionMaterialFunctionCall_8, output=Primary}
    Alpha {from=MaterialExpressionSaturate_2}
   ]
   properties {materialFunction=&Lerp_3Color}
  }
  MaterialExpressionMaterialFunctionCall_8 {
   class=MaterialExpressionMaterialFunctionCall
   inputs [CustomUPosition {from=""}]
   properties {materialFunction=/SparksStageShow/Systems/StageManager/FXManager/MaterialFunctions/MF_MojoStageColors}
  }
  MaterialExpressionMultiply_0 {
   class=MaterialExpressionMultiply
   inputs [
    A {from=MaterialExpressionRayTracingQualitySwitch_0}
    B {from=MaterialExpressionStaticSwitchParameter_0}
   ]
  }
  MaterialExpressionMultiply_1 {
   class=MaterialExpressionMultiply
   inputs [
    A {from=MaterialExpressionMultiply_0}
    B {from=MaterialExpressionShadingPathSwitch_0}
   ]
  }
  MaterialExpressionMultiply_10 {
   class=MaterialExpressionMultiply
   inputs [
    A {from=MaterialExpressionTime_0}
    B {from=MaterialExpressionScalarParameter_5}
   ]
  }
  MaterialExpressionMultiply_11 {
   class=MaterialExpressionMultiply
   inputs [
    A {from=MaterialExpressionTime_3}
    B {from=MaterialExpressionScalarParameter_14}
   ]
  }
  MaterialExpressionMultiply_12 {
   class=MaterialExpressionMultiply
   inputs [
    A {from=MaterialExpressionVectorParameter_1, output=R}
    B {from=MaterialExpressionVectorParameter_11, output=RGB}
   ]
  }
  MaterialExpressionMultiply_13 {
   class=MaterialExpressionMultiply
   inputs [
    A {from=MaterialExpressionVectorParameter_1, output=G}
    B {from=MaterialExpressionVectorParameter_13, output=RGB}
   ]
  }
  MaterialExpressionMultiply_14 {
   class=MaterialExpressionMultiply
   inputs [
    A {from=MaterialExpressionVectorParameter_1, output=B}
    B {from=MaterialExpressionVectorParameter_12, output=RGB}
   ]
  }
  MaterialExpressionMultiply_15 {
   class=MaterialExpressionMultiply
   inputs [
    A {from=MaterialExpressionVectorParameter_1, output=A}
    B {from=MaterialExpressionVectorParameter_14, output=RGB}
   ]
  }
  MaterialExpressionMultiply_16 {
   class=MaterialExpressionMultiply
   inputs [
    A {from=MaterialExpressionVectorParameter_2, output=R}
    B {from=MaterialExpressionVectorParameter_8, output=RGB}
   ]
  }
  MaterialExpressionMultiply_17 {
   class=MaterialExpressionMultiply
   inputs [
    A {from=MaterialExpressionVectorParameter_2, output=G}
    B {from=MaterialExpressionVectorParameter_10, output=RGB}
   ]
  }
  MaterialExpressionMultiply_18 {
   class=MaterialExpressionMultiply
   inputs [
    A {from=MaterialExpressionVectorParameter_2, output=B}
    B {from=MaterialExpressionVectorParameter_9, output=RGB}
   ]
  }
  MaterialExpressionMultiply_2 {
   class=MaterialExpressionMultiply
   inputs [
    A {from=MaterialExpressionReroute_0}
    B {from=MaterialExpressionScalarParameter_6}
   ]
  }
  MaterialExpressionMultiply_3 {
   class=MaterialExpressionMultiply
   inputs [
    A {from=MaterialExpressionAdd_16}
    B {from=MaterialExpressionScalarParameter_1}
   ]
  }
  MaterialExpressionMultiply_4 {
   class=MaterialExpressionMultiply
   inputs [
    A {from=MaterialExpressionReroute_0}
    B {from=MaterialExpressionScalarParameter_8}
   ]
  }
  MaterialExpressionMultiply_5 {
   class=MaterialExpressionMultiply
   inputs [
    A {from=MaterialExpressionReroute_0}
    B {from=MaterialExpressionScalarParameter_9}
   ]
  }
  MaterialExpressionMultiply_6 {
   class=MaterialExpressionMultiply
   inputs [
    A {from=MaterialExpressionScalarParameter_4}
    B {from=MaterialExpressionScalarParameter_10}
   ]
  }
  MaterialExpressionMultiply_7 {
   class=MaterialExpressionMultiply
   inputs [
    A {from=MaterialExpressionScalarParameter_4}
    B {from=MaterialExpressionScalarParameter_11}
   ]
  }
  MaterialExpressionMultiply_8 {
   class=MaterialExpressionMultiply
   inputs [
    A {from=MaterialExpressionNamedRerouteUsage_16}
    B {from=MaterialExpressionMultiply_1}
   ]
  }
  MaterialExpressionMultiply_9 {
   class=MaterialExpressionMultiply
   inputs [
    A {from=MaterialExpressionStaticSwitchParameter_4}
    B {from=MaterialExpressionScalarParameter_2}
   ]
   properties {constB=4}
  }
  MaterialExpressionNamedRerouteDeclaration_0 {
   class=MaterialExpressionNamedRerouteDeclaration
   inputs [None {from=MaterialExpressionStep_0}]
   properties {name=Under 0.25, nodeColor=<rgba 0.0312 1 0 1>}
  }
  MaterialExpressionNamedRerouteDeclaration_1 {
   class=MaterialExpressionNamedRerouteDeclaration
   inputs [None {from=MaterialExpressionSubtract_3}]
   properties {name=Between 0.25 and 0.5, nodeColor=<rgba 1 0 0.3516 1>}
  }
  MaterialExpressionNamedRerouteDeclaration_11 {
   class=MaterialExpressionNamedRerouteDeclaration
   inputs [None {from=MaterialExpressionOneMinus_3}]
   properties {name=Over 0.666, nodeColor=<rgba 0.7578 0 1 1>}
  }
  MaterialExpressionNamedRerouteDeclaration_12 {
   class=MaterialExpressionNamedRerouteDeclaration
   inputs [None {from=MaterialExpressionStep_9}]
   properties {name=Under 0.2, nodeColor=<rgba 0 1 0.4375 1>}
  }
  MaterialExpressionNamedRerouteDeclaration_13 {
   class=MaterialExpressionNamedRerouteDeclaration
   inputs [None {from=MaterialExpressionSubtract_14}]
   properties {name=Between 0.4 and 0.6, nodeColor=<rgba 1 0.1172 0 1>}
  }
  MaterialExpressionNamedRerouteDeclaration_14 {
   class=MaterialExpressionNamedRerouteDeclaration
   inputs [None {from=MaterialExpressionSubtract_15}]
   properties {name=Between 0.6 and 0.8, nodeColor=<rgba 0 0.2031 1 1>}
  }
  MaterialExpressionNamedRerouteDeclaration_15 {
   class=MaterialExpressionNamedRerouteDeclaration
   inputs [None {from=MaterialExpressionSubtract_13}]
   properties {name=Between 0.2 and 0.4, nodeColor=<rgba 0.5234 1 0 1>}
  }
  MaterialExpressionNamedRerouteDeclaration_16 {
   class=MaterialExpressionNamedRerouteDeclaration
   inputs [None {from=MaterialExpressionOneMinus_4}]
   properties {name=Over 0.8, nodeColor=<rgba 1 0 0.8437 1>}
  }
  MaterialExpressionNamedRerouteDeclaration_17 {
   class=MaterialExpressionNamedRerouteDeclaration
   inputs [None {from=MaterialExpressionFloor_1}]
   properties {name=Animation Time, nodeColor=<rgba 0 1 0.8359 1>}
  }
  MaterialExpressionNamedRerouteDeclaration_18 {
   class=MaterialExpressionNamedRerouteDeclaration
   inputs [None {from=MaterialExpressionSwitch_3}]
   properties {name=Vert Alpha Animation, nodeColor=<rgba 1 0.1875 0 1>}
  }
  MaterialExpressionNamedRerouteDeclaration_2 {
   class=MaterialExpressionNamedRerouteDeclaration
   inputs [None {from=MaterialExpressionSubtract_8}]
   properties {name=Between 0.5 and 0.75, nodeColor=<rgba 0 0.6719 1 1>}
  }
  MaterialExpressionNamedRerouteDeclaration_3 {
   class=MaterialExpressionNamedRerouteDeclaration
   inputs [None {from=MaterialExpressionOneMinus_0}]
   properties {name=Over 0.75, nodeColor=<rgba 0.9922 1 0 1>}
  }
  MaterialExpressionNamedRerouteDeclaration_8 {
   class=MaterialExpressionNamedRerouteDeclaration
   inputs [None {from=MaterialExpressionStep_6}]
   properties {name=Under 0.33, nodeColor=<rgba 1 0 0.2812 1>}
  }
  MaterialExpressionNamedRerouteDeclaration_9 {
   class=MaterialExpressionNamedRerouteDeclaration
   inputs [None {from=MaterialExpressionSubtract_11}]
   properties {name=Between 0.333 and 0.666, nodeColor=<rgba 0 0.6016 1 1>}
  }
  MaterialExpressionNamedRerouteUsage_0 {class=MaterialExpressionNamedRerouteUsage}
  MaterialExpressionNamedRerouteUsage_1 {class=MaterialExpressionNamedRerouteUsage}
  MaterialExpressionNamedRerouteUsage_10 {class=MaterialExpressionNamedRerouteUsage}
  MaterialExpressionNamedRerouteUsage_11 {class=MaterialExpressionNamedRerouteUsage}
  MaterialExpressionNamedRerouteUsage_12 {class=MaterialExpressionNamedRerouteUsage}
  MaterialExpressionNamedRerouteUsage_13 {class=MaterialExpressionNamedRerouteUsage}
  MaterialExpressionNamedRerouteUsage_14 {class=MaterialExpressionNamedRerouteUsage}
  MaterialExpressionNamedRerouteUsage_15 {class=MaterialExpressionNamedRerouteUsage}
  MaterialExpressionNamedRerouteUsage_16 {class=MaterialExpressionNamedRerouteUsage}
  MaterialExpressionNamedRerouteUsage_2 {class=MaterialExpressionNamedRerouteUsage}
  MaterialExpressionNamedRerouteUsage_3 {class=MaterialExpressionNamedRerouteUsage}
  MaterialExpressionNamedRerouteUsage_4 {class=MaterialExpressionNamedRerouteUsage}
  MaterialExpressionNamedRerouteUsage_5 {class=MaterialExpressionNamedRerouteUsage}
  MaterialExpressionNamedRerouteUsage_6 {class=MaterialExpressionNamedRerouteUsage}
  MaterialExpressionNamedRerouteUsage_8 {class=MaterialExpressionNamedRerouteUsage}
  MaterialExpressionNamedRerouteUsage_9 {class=MaterialExpressionNamedRerouteUsage}
  MaterialExpressionObjectPositionWS_1 {class=MaterialExpressionObjectPositionWS, inputs [], properties {}}
  MaterialExpressionOneMinus_0 {class=MaterialExpressionOneMinus, inputs [None {from=MaterialExpressionStep_2}]}
  MaterialExpressionOneMinus_3 {class=MaterialExpressionOneMinus, inputs [None {from=MaterialExpressionStep_7}]}
  MaterialExpressionOneMinus_4 {class=MaterialExpressionOneMinus, inputs [None {from=MaterialExpressionStep_12}]}
  MaterialExpressionQualitySwitch_0 {
   class=MaterialExpressionQualitySwitch
   inputs [
    Default {from=MaterialExpressionMultiply_4, output=""}
    Low {from=MaterialExpressionMultiply_4, output=""}
    High {from="", output=""}
    Medium {from=MaterialExpressionMultiply_5, output=""}
    Epic {from="", output=""}
   ]
   properties {}
  }
  MaterialExpressionQualitySwitch_1 {
   class=MaterialExpressionQualitySwitch
   inputs [
    Default {from=MaterialExpressionMultiply_6, output=""}
    Low {from=MaterialExpressionMultiply_6, output=""}
    High {from="", output=""}
    Medium {from=MaterialExpressionMultiply_7, output=""}
    Epic {from="", output=""}
   ]
   properties {}
  }
  MaterialExpressionRayTracingQualitySwitch_0 {
   class=MaterialExpressionRayTracingQualitySwitch
   inputs [
    Normal {from=MaterialExpressionMaterialFunctionCall_0, output=Fresnel}
    RayTraced {from=MaterialExpressionScalarParameter_7, output=""}
   ]
   properties {}
  }
  MaterialExpressionReroute_0 {class=MaterialExpressionReroute, inputs [None {from=MaterialExpressionStaticSwitchParameter_1}]}
  MaterialExpressionReroute_1 {class=MaterialExpressionReroute, inputs [None {from=MaterialExpressionStaticSwitch_0}]}
  MaterialExpressionSaturate_2 {class=MaterialExpressionSaturate, inputs [None {from=MaterialExpressionScalarParameter_12}]}
  MaterialExpressionSaturate_5 {class=MaterialExpressionSaturate, inputs [None {from=MaterialExpressionAdd_0}]}
  MaterialExpressionSaturate_6 {class=MaterialExpressionSaturate, inputs [None {from=MaterialExpressionAdd_5}]}
  MaterialExpressionSaturate_7 {class=MaterialExpressionSaturate, inputs [None {from=MaterialExpressionAdd_6}]}
  MaterialExpressionScalarParameter_0 {class=MaterialExpressionScalarParameter, properties {defaultValue=1, parameterName=Metallic, sliderMax=1}}
  MaterialExpressionScalarParameter_1 {class=MaterialExpressionScalarParameter, properties {defaultValue=1, parameterName=LightsOn, sliderMax=1}}
  MaterialExpressionScalarParameter_10 {class=MaterialExpressionScalarParameter, properties {defaultValue=1, parameterName=GlowFresnelExponentScale_Mobile}}
  MaterialExpressionScalarParameter_11 {class=MaterialExpressionScalarParameter, properties {defaultValue=1, parameterName=GlowFresnelExponentScale_Switch}}
  MaterialExpressionScalarParameter_12 {class=MaterialExpressionScalarParameter, properties {parameterName=StageShowColorSelect}}
  MaterialExpressionScalarParameter_13 {class=MaterialExpressionScalarParameter, properties {defaultValue=0.948, parameterName=Roughness_Mobile, sliderMax=1}}
  MaterialExpressionScalarParameter_14 {class=MaterialExpressionScalarParameter, properties {defaultValue=20, parameterName=FlickerSpeed}}
  MaterialExpressionScalarParameter_15 {
   class=MaterialExpressionScalarParameter
   properties {defaultValue=3, group=Vertex Alpha Animation, parameterName=Number of Frames, sliderMax=4, sliderMin=2}
  }
  MaterialExpressionScalarParameter_2 {class=MaterialExpressionScalarParameter, properties {defaultValue=3, group=Vertex Alpha Animation, parameterName=Number of Frames}}
  MaterialExpressionScalarParameter_3 {class=MaterialExpressionScalarParameter, properties {defaultValue=0.148, parameterName=Roughness, sliderMax=1}}
  MaterialExpressionScalarParameter_4 {class=MaterialExpressionScalarParameter, properties {defaultValue=10, parameterName=GlowFresnelExponent}}
  MaterialExpressionScalarParameter_5 {class=MaterialExpressionScalarParameter, properties {defaultValue=1, group=Vertex Alpha Animation, parameterName=Speed}}
  MaterialExpressionScalarParameter_6 {class=MaterialExpressionScalarParameter, properties {defaultValue=25, parameterName=Brightness}}
  MaterialExpressionScalarParameter_7 {class=MaterialExpressionScalarParameter, properties {defaultValue=0.1, parameterName=LumenEmissiveMult}}
  MaterialExpressionScalarParameter_8 {class=MaterialExpressionScalarParameter, properties {defaultValue=10, parameterName=Brightness_Mobile}}
  MaterialExpressionScalarParameter_9 {class=MaterialExpressionScalarParameter, properties {defaultValue=10, parameterName=Brightness_Switch}}
  MaterialExpressionShadingPathSwitch_0 {
   class=MaterialExpressionShadingPathSwitch
   inputs [
    Default {from=MaterialExpressionMultiply_2, output=""}
    Deferred {from=MaterialExpressionMultiply_2, output=""}
    Forward {from="", output=""}
    Mobile {from=MaterialExpressionQualitySwitch_0, output=""}
   ]
   properties {}
  }
  MaterialExpressionShadingPathSwitch_1 {
   class=MaterialExpressionShadingPathSwitch
   inputs [
    Default {from=MaterialExpressionScalarParameter_4, output=""}
    Deferred {from=MaterialExpressionScalarParameter_4, output=""}
    Forward {from="", output=""}
    Mobile {from=MaterialExpressionQualitySwitch_1, output=""}
   ]
   properties {}
  }
  MaterialExpressionShadingPathSwitch_2 {
   class=MaterialExpressionShadingPathSwitch
   inputs [
    Default {from=MaterialExpressionScalarParameter_3, output=""}
    Deferred {from=MaterialExpressionScalarParameter_3, output=""}
    Forward {from="", output=""}
    Mobile {from=MaterialExpressionScalarParameter_13, output=""}
   ]
   properties {}
  }
  MaterialExpressionSine_4 {class=MaterialExpressionSine, inputs [None {from=MaterialExpressionReroute_1}]}
  MaterialExpressionSine_5 {class=MaterialExpressionSine, inputs [None {from=MaterialExpressionReroute_1}], properties {period=0.8095}}
  MaterialExpressionSine_6 {class=MaterialExpressionSine, inputs [None {from=MaterialExpressionReroute_1}], properties {period=0.1807}}
  MaterialExpressionStaticBoolParameter_0 {class=MaterialExpressionStaticBoolParameter, inputs [], properties {defaultValue=true, parameterName=Sync Flicker to Beat}}
  MaterialExpressionStaticSwitchParameter_0 {
   class=MaterialExpressionStaticSwitchParameter
   inputs [
    True {from=MaterialExpressionMultiply_3, output=""}
    False {from=MaterialExpressionScalarParameter_1, output=""}
   ]
   properties {defaultValue=true, parameterName=Flicker}
  }
  MaterialExpressionStaticSwitchParameter_1 {
   class=MaterialExpressionStaticSwitchParameter
   inputs [
    True {from=MaterialExpressionLinearInterpolate_7, output=""}
    False {from=MaterialExpressionVectorParameter_0, output=RGB}
   ]
   properties {parameterName=Use VertexColor as channel masks}
  }
  MaterialExpressionStaticSwitchParameter_2 {
   class=MaterialExpressionStaticSwitchParameter
   inputs [
    True {from=MaterialExpressionMaterialFunctionCall_2, output=Result}
    False {from=MaterialExpressionConstant3Vector_2, output=""}
   ]
   properties {parameterName=UseStageShowColors}
  }
  MaterialExpressionStaticSwitchParameter_3 {
   class=MaterialExpressionStaticSwitchParameter
   inputs [
    True {from=MaterialExpressionMultiply_8, output=""}
    False {from=MaterialExpressionMultiply_1, output=""}
   ]
   properties {group=Vertex Alpha Animation, parameterName=Use Vert Alpha Animation}
  }
  MaterialExpressionStaticSwitchParameter_4 {
   class=MaterialExpressionStaticSwitchParameter
   inputs [
    True {from=MaterialExpressionCollectionParameter_0, output=""}
    False {from=MaterialExpressionFrac_1, output=""}
   ]
   properties {defaultValue=true, parameterName=_Sync Animation to Measure}
  }
  MaterialExpressionStaticSwitch_0 {
   class=MaterialExpressionStaticSwitch
   inputs [
    True {from=MaterialExpressionCollectionParameter_1, output=""}
    False {from=MaterialExpressionAdd_2, output=""}
    Value {from=MaterialExpressionStaticBoolParameter_0, output=""}
   ]
   properties {}
  }
  MaterialExpressionStep_0 {
   class=MaterialExpressionStep
   inputs [
    Y {from=MaterialExpressionVertexColor_2, output=A}
    X {from="", output=""}
   ]
   properties {constX=0.25}
  }
  MaterialExpressionStep_1 {
   class=MaterialExpressionStep
   inputs [
    Y {from=MaterialExpressionVertexColor_2, output=A}
    X {from="", output=""}
   ]
   properties {constX=0.5}
  }
  MaterialExpressionStep_10 {
   class=MaterialExpressionStep
   inputs [
    Y {from=MaterialExpressionVertexColor_5, output=A}
    X {from="", output=""}
   ]
   properties {constX=0.4}
  }
  MaterialExpressionStep_11 {
   class=MaterialExpressionStep
   inputs [
    Y {from=MaterialExpressionVertexColor_5, output=A}
    X {from="", output=""}
   ]
   properties {constX=0.6}
  }
  MaterialExpressionStep_12 {
   class=MaterialExpressionStep
   inputs [
    Y {from=MaterialExpressionVertexColor_5, output=A}
    X {from="", output=""}
   ]
   properties {constX=0.8}
  }
  MaterialExpressionStep_2 {
   class=MaterialExpressionStep
   inputs [
    Y {from=MaterialExpressionVertexColor_2, output=A}
    X {from="", output=""}
   ]
   properties {constX=0.75}
  }
  MaterialExpressionStep_6 {
   class=MaterialExpressionStep
   inputs [
    Y {from=MaterialExpressionVertexColor_4, output=A}
    X {from="", output=""}
   ]
   properties {constX=0.333}
  }
  MaterialExpressionStep_7 {
   class=MaterialExpressionStep
   inputs [
    Y {from=MaterialExpressionVertexColor_4, output=A}
    X {from="", output=""}
   ]
   properties {constX=0.666}
  }
  MaterialExpressionStep_9 {
   class=MaterialExpressionStep
   inputs [
    Y {from=MaterialExpressionVertexColor_5, output=A}
    X {from="", output=""}
   ]
   properties {constX=0.2}
  }
  MaterialExpressionSubtract_11 {
   class=MaterialExpressionSubtract
   inputs [
    A {from=MaterialExpressionStep_7}
    B {from=MaterialExpressionStep_6}
   ]
  }
  MaterialExpressionSubtract_13 {
   class=MaterialExpressionSubtract
   inputs [
    A {from=MaterialExpressionStep_10}
    B {from=MaterialExpressionStep_9}
   ]
  }
  MaterialExpressionSubtract_14 {
   class=MaterialExpressionSubtract
   inputs [
    A {from=MaterialExpressionStep_11}
    B {from=MaterialExpressionStep_10}
   ]
  }
  MaterialExpressionSubtract_15 {
   class=MaterialExpressionSubtract
   inputs [
    A {from=MaterialExpressionStep_12}
    B {from=MaterialExpressionStep_11}
   ]
  }
  MaterialExpressionSubtract_16 {
   class=MaterialExpressionSubtract
   inputs [
    A {from=MaterialExpressionScalarParameter_15}
    B {from=""}
   ]
   properties {constB=2}
  }
  MaterialExpressionSubtract_3 {
   class=MaterialExpressionSubtract
   inputs [
    A {from=MaterialExpressionStep_1}
    B {from=MaterialExpressionStep_0}
   ]
  }
  MaterialExpressionSubtract_8 {
   class=MaterialExpressionSubtract
   inputs [
    A {from=MaterialExpressionStep_2}
    B {from=MaterialExpressionStep_1}
   ]
  }
  MaterialExpressionSwitch_0 {
   class=MaterialExpressionSwitch
   inputs [
    SwitchValue {from=MaterialExpressionNamedRerouteUsage_3, output=""}
    Default {from=MaterialExpressionNamedRerouteUsage_5, output=""}
    None {from=MaterialExpressionNamedRerouteUsage_5, output=""}
    None {from=MaterialExpressionNamedRerouteUsage_9, output=""}
    None {from=MaterialExpressionNamedRerouteUsage_10, output=""}
   ]
   properties {
    constDefault=1
    constSwitchValue=1
    inputs [
     {
      inputName=None
      input {
       expression=&self:MaterialExpressionNamedRerouteUsage_5
       inputName=None
       mask=0
       maskA=0
       maskB=0
       maskG=0
       maskR=0
       outputIndex=0
      }
     }
     {
      inputName=None
      input {
       expression=&self:MaterialExpressionNamedRerouteUsage_9
       inputName=None
       mask=0
       maskA=0
       maskB=0
       maskG=0
       maskR=0
       outputIndex=0
      }
     }
     {
      inputName=None
      input {
       expression=&self:MaterialExpressionNamedRerouteUsage_10
       inputName=None
       mask=0
       maskA=0
       maskB=0
       maskG=0
       maskR=0
       outputIndex=0
      }
     }
    ]
   }
  }
  MaterialExpressionSwitch_1 {
   class=MaterialExpressionSwitch
   inputs [
    SwitchValue {from=MaterialExpressionNamedRerouteUsage_6, output=""}
    Default {from=MaterialExpressionNamedRerouteUsage_8, output=""}
    None {from=MaterialExpressionNamedRerouteUsage_8, output=""}
    None {from=MaterialExpressionNamedRerouteUsage_15, output=""}
    None {from=MaterialExpressionNamedRerouteUsage_14, output=""}
    None {from=MaterialExpressionNamedRerouteUsage_13, output=""}
   ]
   properties {
    constDefault=1
    constSwitchValue=1
    inputs [
     {
      inputName=None
      input {
       expression=&self:MaterialExpressionNamedRerouteUsage_8
       inputName=None
       mask=0
       maskA=0
       maskB=0
       maskG=0
       maskR=0
       outputIndex=0
      }
     }
     {
      inputName=None
      input {
       expression=&self:MaterialExpressionNamedRerouteUsage_15
       inputName=None
       mask=0
       maskA=0
       maskB=0
       maskG=0
       maskR=0
       outputIndex=0
      }
     }
     {
      inputName=None
      input {
       expression=&self:MaterialExpressionNamedRerouteUsage_14
       inputName=None
       mask=0
       maskA=0
       maskB=0
       maskG=0
       maskR=0
       outputIndex=0
      }
     }
     {
      inputName=None
      input {
       expression=&self:MaterialExpressionNamedRerouteUsage_13
       inputName=None
       mask=0
       maskA=0
       maskB=0
       maskG=0
       maskR=0
       outputIndex=0
      }
     }
    ]
   }
  }
  MaterialExpressionSwitch_2 {
   class=MaterialExpressionSwitch
   inputs [
    SwitchValue {from=MaterialExpressionNamedRerouteUsage_4, output=""}
    Default {from=MaterialExpressionNamedRerouteUsage_0, output=""}
    None {from=MaterialExpressionNamedRerouteUsage_0, output=""}
    None {from=MaterialExpressionNamedRerouteUsage_1, output=""}
   ]
   properties {
    constDefault=1
    constSwitchValue=1
    inputs [
     {
      inputName=None
      input {
       expression=&self:MaterialExpressionNamedRerouteUsage_0
       inputName=None
       mask=0
       maskA=0
       maskB=0
       maskG=0
       maskR=0
       outputIndex=0
      }
     }
     {
      inputName=None
      input {
       expression=&self:MaterialExpressionNamedRerouteUsage_1
       inputName=None
       mask=0
       maskA=0
       maskB=0
       maskG=0
       maskR=0
       outputIndex=0
      }
     }
    ]
   }
  }
  MaterialExpressionSwitch_3 {
   class=MaterialExpressionSwitch
   inputs [
    SwitchValue {from=MaterialExpressionSubtract_16, output=""}
    Default {from=MaterialExpressionSaturate_7, output=""}
    None {from=MaterialExpressionSaturate_5, output=""}
    None {from=MaterialExpressionSaturate_6, output=""}
   ]
   properties {inputs [
     {
      inputName=None
      input {
       expression=&self:MaterialExpressionSaturate_5
       inputName=None
       mask=0
       maskA=0
       maskB=0
       maskG=0
       maskR=0
       outputIndex=0
      }
     }
     {
      inputName=None
      input {
       expression=&self:MaterialExpressionSaturate_6
       inputName=None
       mask=0
       maskA=0
       maskB=0
       maskG=0
       maskR=0
       outputIndex=0
      }
     }
    ]}
  }
  MaterialExpressionTime_0 {class=MaterialExpressionTime}
  MaterialExpressionTime_3 {class=MaterialExpressionTime}
  MaterialExpressionTransform_0 {class=MaterialExpressionTransform, inputs [None {from=MaterialExpressionConstant3Vector_0}]}
  MaterialExpressionVectorParameter_0 {class=MaterialExpressionVectorParameter, properties {parameterName=Color, defaultValue=<rgba 1 0 0.4825 1>}}
  MaterialExpressionVectorParameter_1 {class=MaterialExpressionVectorParameter, properties {parameterName=FlickerRandomMask_A, defaultValue=<rgba 1 1 1 1>}}
  MaterialExpressionVectorParameter_10 {class=MaterialExpressionVectorParameter, properties {desc=Red, parameterName=Color_RB, defaultValue=<rgba 1 0.0982 0.0285 0>}}
  MaterialExpressionVectorParameter_11 {class=MaterialExpressionVectorParameter, properties {desc=Warm White, parameterName=Color_None, defaultValue=<rgba 1 0.8495 0.682 0>}}
  MaterialExpressionVectorParameter_12 {class=MaterialExpressionVectorParameter, properties {desc=Magenta, parameterName=Color_G, defaultValue=<rgba 1 0.8495 0.682 0>}}
  MaterialExpressionVectorParameter_13 {class=MaterialExpressionVectorParameter, properties {desc=Light blue, parameterName=Color_B, defaultValue=<rgba 0.0142 0.3663 1 0>}}
  MaterialExpressionVectorParameter_14 {class=MaterialExpressionVectorParameter, properties {desc=Peach, parameterName=Color_GB, defaultValue=<rgba 1 0.5904 0.6403 0>}}
  MaterialExpressionVectorParameter_2 {class=MaterialExpressionVectorParameter, properties {parameterName=FlickerRandomMask_B, defaultValue=<rgba 1 1 1 1>}}
  MaterialExpressionVectorParameter_8 {class=MaterialExpressionVectorParameter, properties {desc=Turquoise, parameterName=Color_R, defaultValue=<rgba 0 1 0.6162 0>}}
  MaterialExpressionVectorParameter_9 {class=MaterialExpressionVectorParameter, properties {desc=Yellow, parameterName=Color_RG, defaultValue=<rgba 1 0.8702 0 0>}}
  MaterialExpressionVertexColor_1 {class=MaterialExpressionVertexColor, inputs [], properties {}}
  MaterialExpressionVertexColor_2 {class=MaterialExpressionVertexColor, inputs [], properties {}}
  MaterialExpressionVertexColor_4 {class=MaterialExpressionVertexColor, inputs [], properties {}}
  MaterialExpressionVertexColor_5 {class=MaterialExpressionVertexColor, inputs [], properties {}}
 }
 properties {bUsedWithNanite=true, pixelDepthOffsetMode=PDOM_Legacy, usedShadingModels=Default Lit}
}
